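# Post-process style Anim Blueprint for the female Barium Jacket player body, adding deformation and secondary motion to the parent mesh's pose.
# Copied pose runs through two Control Rigs, the second flagged during the front-end idle montage.
# AnimDynamics springs then swing two zipper bones, the second constraint-anchored, and both lower pant legs.
# Without an extra-part shoe mesh, left-foot velocity drives material WPO, with stronger noise while skydiving.
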AnimBlueprint /Game/Characters/Player/Female/Medium/Bodies/F_MED_Barium_Jacket/Meshes/F_MED_Barium_Jacket_AnimBP
target_skeleton=&/F_MED_Barium_Jacket_Skeleton
graphs [
 AnimGraph {nodes [
   n1 = Control Rig(Source=Copy Pose From Mesh() {bUseAttachedParent=true}) {
    alphaBoolBlend {blendOption=Linear}
    controlRigAssetReference {blueprintRigClass=&/F_MED_Barium_Jacket_Deformation_CtrlRig.F_MED_Barium_Jacket_Deformation_CtrlRig_C}
   }
   n2 = Control Rig!(Source=@n1, Alpha, is_playing_character_idle:bool=Get Is Playing Character Idle() -> (Is Playing Character Idle:bool)) -> (Pose) {
    alphaBoolBlend {blendOption=Linear}
    controlRigAssetReference {blueprintRigClass=&/F_MED_Barium_Jacket_CtrlRig.F_MED_Barium_Jacket_CtrlRig_C}
   }
   n3 = Anim Dynamics - Bone: dyn_zipper_01(ComponentPose=Local To Component(LocalPose=@n2), Alpha=1, ExternalForce="0,0,0") {
    angularDampingOverride=0.99
    angularSpringConstant=3000
    bAngularSpring=true
    bOverrideAngularDamping=true
    bUseGravityOverride=true
    alphaBoolBlend {blendOption=Linear}
    boundBone {boneName=dyn_zipper_01}
    componentAppliedLinearAccClamp=<3000 3000 3000>
    componentLinearAccScale=<0.1 0.1 0.1>
    componentLinearVelScale=<0.9 0.9 0.9>
    gravityOverride {z=-980}
    physicsBodyDefinitions [{
      boundBone {boneName=dyn_zipper_01}
      boxExtents=<10 10 10>
      constraintSetup {angularLimitsMax=<0 5 5>, angularLimitsMin=<0 -5 -20>, angularTarget=<1 0 0>}
      localJointOffset=<2 0 0>
     }]
   }
   n4 = "Constraint - dyn_zipper_02_mstr (PPPP)"!(
    ComponentPose=@n3
    Alpha=1
    ConstraintWeights_0=0.0077
    ConstraintWeights_1:float=0.2327
    ConstraintWeights_2:float=0.0077
    ConstraintWeights_3:float=0.7519
   ) -> (Pose) {
    alphaBoolBlend {blendOption=Linear}
    boneToModify {boneName=dyn_zipper_02_mstr}
    constraintSetup [
     {targetBone {boneName=deform_pec_r}}
     {targetBone {boneName=spine_05}}
     {targetBone {boneName=deform_pec_l}}
     {targetBone {boneName=spine_04}}
    ]
    constraintWeights [1,1,1,1]
   }
   n5 = Anim Dynamics - Bone: dyn_zipper_02(ComponentPose=@n4, Alpha=1, ExternalForce="0,0,0") {
    angularDampingOverride=0.99
    angularSpringConstant=3000
    bAngularSpring=true
    bOverrideAngularDamping=true
    bUseGravityOverride=true
    alphaBoolBlend {blendOption=Linear}
    boundBone {boneName=dyn_zipper_02}
    componentAppliedLinearAccClamp=<3000 3000 3000>
    componentLinearAccScale=<0.1 0.1 0.1>
    componentLinearVelScale=<0.9 0.9 0.9>
    gravityOverride {z=-980}
    physicsBodyDefinitions [{
      boundBone {boneName=dyn_zipper_02}
      boxExtents=<10 10 10>
      constraintSetup {angularLimitsMax=<0 5 5>, angularLimitsMin=<0 -5 -20>, angularTarget=<1 0 0>}
      localJointOffset=<2 0 0>
     }]
   }
   n6 = Anim Dynamics - Bone: dyn_lower_pants_l(ComponentPose=@n5, Alpha=1, ExternalForce="0,0,0") {
    bLinearSpring=true
    bOverrideLinearDamping=true
    bUseGravityOverride=true
    linearDampingOverride=1
    linearSpringConstant=175
    simulationSpace=RootRelative
    alphaBoolBlend {blendOption=Linear}
    boundBone {boneName=dyn_lower_pants_l}
    componentAppliedLinearAccClamp=<3000 3000 3000>
    componentLinearAccScale=<0.1 0.1 0.1>
    componentLinearVelScale=<0.9 0.9 0.9>
    physicsBodyDefinitions [{
      boundBone {boneName=dyn_lower_pants_l}
      boxExtents=<60 60 60>
      constraintSetup {angularLimitsMax=<0 0 0>, angularLimitsMin=<0 0 0>, angularTarget=<0 0 0>, linearAxesMax=<5 4 5>, linearAxesMin=<-1.5 -4 -5>}
      localJointOffset=<0 0 0>
     }]
    relativeSpaceBone {boneName=foot_l}
   }
   n7 = Anim Dynamics - Bone: dyn_lower_pants_r(ComponentPose=@n6, Alpha=1, ExternalForce="0,0,0") {
    bLinearSpring=true
    bOverrideLinearDamping=true
    bUseGravityOverride=true
    linearDampingOverride=1
    linearSpringConstant=175
    simulationSpace=RootRelative
    alphaBoolBlend {blendOption=Linear}
    boundBone {boneName=dyn_lower_pants_r}
    componentAppliedLinearAccClamp=<3000 3000 3000>
    componentLinearAccScale=<0.1 0.1 0.1>
    componentLinearVelScale=<0.9 0.9 0.9>
    physicsBodyDefinitions [{
      boundBone {boneName=dyn_lower_pants_r}
      boxExtents=<60 60 60>
      constraintSetup {angularLimitsMax=<0 0 0>, angularLimitsMin=<0 0 0>, angularTarget=<0 0 0>, linearAxesMax=<1.5 4 5>, linearAxesMin=<-5 -4 -5>}
      localJointOffset=<0 0 0>
     }]
    relativeSpaceBone {boneName=foot_l}
   }
   Output Pose(Result=Component To Local(ComponentPose=@n7)) {layerGroup=None}
  ]}
 EventGraph {nodes [
   n1 = Event BlueprintUpdateAnimation()
   > n2 = Sequence!(execute) -> (then_0, then_1, then_2:Exec)
   > Set looking_down(looking_down:double=GetAlphaFromReferenceTransformDelta(
     SkeletalMeshComponent=Get skeletal_mesh() -> (skeletal_mesh:Skeletal Mesh Component Object Reference)
     BoneName=head
     BaseName=spine_05
     DeltaComponentType=QuaternionTwist
     InputRange=(X=0,Y=-22)
     TwistAxis="0,0,1"
    )) -> (Output_Get:double)
   Is Valid(exec=@n2.then_1, InputObject=Get playerPawn() -> (playerPawn:Fort Player Pawn Athena Object Reference))
   > n3 = Get!(execute:Exec) -> (then:Exec, else:Exec, ShoeMesh:Skeletal Mesh Component Budgeted Object Reference)
   > n4 = Is Valid(InputObject=Get SkeletalMesh!(self:Skinned Mesh Component Object Reference=@n3) -> (SkeletalMesh:Skeletal Mesh Object Reference))
   n5 = Get MID() -> (MID:Material Instance Dynamic Object Reference)
   n6 = setWPOMaterialVelocity!(
    execute=@n4.Is Not Valid
    self
    materialInstance:Material Instance Dynamic Object Reference=@n5
    skelMesh:Skeletal Mesh Component Object Reference=Get skeletal_mesh()
    boneName:Name=foot_l
    offsetBoneName:Name=calf_l
    offsetBoneAxis:EAxis Enum=Y
    deltaTimeX:double=@n1.DeltaTimeX
    materialVelocityAttribute:Name=Velocity
    previousBoneLocation:Vector=Get previousLFootLocation() -> (previousLFootLocation:Vector)
    previousBoneVelocityVector:Vector=Get previousLFootVelocityVector() -> (previousLFootVelocityVector:Vector)
   ) -> (then, boneVelocity:double, boneVelocityVector:Vector, boneLocation:Vector, boneTransform:Transform)
   > Set previousLFootVelocityVector(previousLFootVelocityVector:Vector=@n6.boneVelocityVector) -> (Output_Get:Vector)
   > Set previousLFootLocation(previousLFootLocation:Vector=@n6.boneLocation) -> (Output_Get:Vector)
   > n7 = Branch(Condition=Get bIsSkyDiving())
   > n8 = Set wpoNoiseStrengthMin(wpoNoiseStrengthMin:double=3) -> (Output_Get:double)
   SetScalarParameterValue(
    execute=[@n9.then, @n8.then]
    self=@n5
    ParameterName=ClothSimNoiseStrengthMin
    Value=Get wpoNoiseStrengthMin() -> (wpoNoiseStrengthMin:double)
   )
   n9 = Set wpoNoiseStrengthMin(execute=@n7.else, wpoNoiseStrengthMin=0)
   SetVectorParameterValue(
    execute=@n4.Is Valid
    self=Get MID()
    ParameterName=Velocity
    Value="To LinearColor (Vector)"!(InVec:Vector=Get empty_vector() -> (empty_vector:Vector)) -> (ReturnValue:Linear Color Structure)
   )
   n10 = GetAnimInstance(self=GetOwnerSkeletalMeshComponent())
   Is Valid(exec=@n2.then_2, InputObject=@n10)
   > n11 = Branch(Condition=Get bIsInFrontEnd())
   > Set Is Playing Character Idle(Is Playing Character Idle:bool=Montage_IsActive!(
     self:Anim Instance Object Reference=@n10
     Montage:Anim Montage Object Reference=/Game/Animation/Game/MainPlayer/Menu/BR/CustomSkins/_s21/BariumF/Barium_Female_Idle_Montage
    ) -> (ReturnValue:bool)) -> (Output_Get:bool)
   Set Is Playing Character Idle(execute=@n11.else, Is Playing Character Idle=false)
   Event BlueprintInitializeAnimation()
   > Set skeletal_mesh(skeletal_mesh:Skeletal Mesh Component Object Reference=GetOwnerSkeletalMeshComponent()) -> (Output_Get:Skeletal Mesh Component Object Reference)
   n12 = TryGetPawnOwner()
   > Is Valid(InputObject=@n12)
   > n13 = Cast To FortPlayerPawnAthena(Object=@n12)
   > Set playerPawn(playerPawn:Fort Player Pawn Athena Object Reference=@n13) -> (Output_Get:Fort Player Pawn Athena Object Reference)
   n14 = Event OnCharacterCustomizationLoaded()
   > n15 = Set SkelMesh(SkelMesh=GetPartSkeletalMeshComponent())
   > n16 = Get Dynamic Material Instance(SkeletalMeshComponent=@n15)
   > Set MID(MID:Material Instance Dynamic Object Reference=@n16) -> (Output_Get:Material Instance Dynamic Object Reference)
   > Set ShoeMesh(ShoeMesh:Skeletal Mesh Component Budgeted Object Reference=Get Array Item with EFort Custom Part Type(Array=@n14.CustomMeshParts, PartType=ExtraPart)) -> (Output_Get:Skeletal Mesh Component Budgeted Object Reference)
  ]}
 calculateMotionVectorForSocket {nodes [
   n1 = calculateMotionVectorForSocket!() -> (then, skelMesh:Scene Component Object Reference, boneName:Name, previousBoneLocation:Vector=~) {is=title K2Node_FunctionEntry}
   n2 = GetSocketTransform(self=@n1, InSocketName=@n1.boneName, TransformSpace=RTS_World)
   n3 = BreakTransform(InTransform=@n2)
   > Return Node!(execute, boneMotionVector:Vector=vector - vector(A=@n3, B=@n1.previousBoneLocation), boneLocation:Vector=@n3, boneTransform:Transform=@n2)
  ]}
 setWPOMaterialVelocity {nodes [
   n1 = setWPOMaterialVelocity!() -> (
    then
    materialInstance:Material Instance Dynamic Object Reference
    skelMesh:Skeletal Mesh Component Object Reference
    boneName:Name
    offsetBoneName:Name
    offsetBoneAxis:EAxis Enum=None
    deltaTimeX:double=~
    materialVelocityAttribute:Name=Velocity
    previousBoneLocation:Vector=~
    previousBoneVelocityVector:Vector=~
   ) {is=title K2Node_FunctionEntry}
   > n2 = calculateMotionVectorForSocket!(
    execute
    self
    skelMesh:Scene Component Object Reference=@n1.skelMesh
    boneName:Name=@n1.boneName
    previousBoneLocation:Vector=@n1.previousBoneLocation
   ) -> (then, boneMotionVector:Vector, boneLocation:Vector, boneTransform:Transform) {is=title K2Node_CallFunction}
   n3 = "Get Unit Axis (Matrix)"!(
    M:"Matrix Structure (by ref)"="To Matrix (Transform)"!(Transform:"Transform (by ref)"=GetSocketTransform(self=@n1.skelMesh, InSocketName=@n1.offsetBoneName, TransformSpace=RTS_World)) -> (ReturnValue:Matrix Structure)
    Axis:EAxis Enum=@n1.offsetBoneAxis
   ) -> (ReturnValue:Vector) {is=title K2Node_CallFunction}
   n4 = "Max (Float)"(A=@n1.deltaTimeX, B=0.01)
   n5 = vector / vector(A:Vector=@n2, B=@n4) -> (ReturnValue:Vector)
   n6 = VInterpTo!(Current:Vector=@n1.previousBoneVelocityVector, Target:Vector=@n5, DeltaTime:float=@n4, InterpSpeed:float=0.75) -> (ReturnValue:Vector)
   > SetVectorParameterValue(self=@n1, ParameterName=@n1.materialVelocityAttribute, Value="To LinearColor (Vector)"(InVec=vector + vector(A=@n3, B=@n6)))
   > Return Node!(
    execute
    boneVelocity:double=Vector Length(A=@n5)
    boneVelocityVector:Vector=@n5
    boneLocation=@n2.boneLocation
    boneTransform=@n2.boneTransform
   )
  ]}
]
variables [
 empty_vector {type=Vector}
 Is Playing Character Idle {type=boolean}
 looking_down {type=number}
 MID {type=MaterialInstanceDynamic}
 playerPawn {type=FortPlayerPawnAthena}
 previousLFootLocation {type=Vector}
 previousLFootVelocityVector {type=Vector}
 ShoeMesh {type=SkeletalMeshComponentBudgeted}
 skeletal_mesh {type=SkeletalMeshComponent}
 SkelMesh {type=SkeletalMeshComponent}
 wpoNoiseStrengthMin {type=number}
]
